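# Sound cue for hard scuff foley on the Mimosa dino cosmetic companion, routed to the creature pets sound class.
# A quality-level switch sends its first two levels to an equal-weight random pick of ten scuff waves and its last two to a pick of only four.
# Playback is quieter and pitched up, with companion foley attenuation and concurrency borrowed from the biped footstep set.
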SoundCue /CosmeticCompanions/Assets/Audio_Shared/Foley/Dino/SC_MImosa_Dino_Scuff
root=QualityLevel_0
nodes [
 {id=QualityLevel_0, children [Random_0,Random_0,Random_1,Random_1]}
 {
  id=Random_0
  children [WavePlayer_0,WavePlayer_1,WavePlayer_2,WavePlayer_3,WavePlayer_4,WavePlayer_5,WavePlayer_6,WavePlayer_7,WavePlayer_8,WavePlayer_9]
  properties {weights [1,1,1,1,1,1,1,1,1,1]}
 }
 {id=Random_1, children [WavePlayer_10, WavePlayer_11, WavePlayer_12, WavePlayer_13], properties {weights [1,1,1,1]}}
 {id=WavePlayer_0, properties {soundWaveAssetPtr=/CosmeticCompanions/Assets/Audio_Shared/Foley/Dino/Wavs/Scuff/Mimosa_Dino_Scuff_Hard_01}}
 {id=WavePlayer_1, properties {soundWaveAssetPtr=/CosmeticCompanions/Assets/Audio_Shared/Foley/Dino/Wavs/Scuff/Mimosa_Dino_Scuff_Hard_02}}
 {id=WavePlayer_10, properties {soundWaveAssetPtr=&Mimosa_Dino_Scuff_Hard_01}}
 {id=WavePlayer_11, properties {soundWaveAssetPtr=&Mimosa_Dino_Scuff_Hard_02}}
 {id=WavePlayer_12, properties {soundWaveAssetPtr=/CosmeticCompanions/Assets/Audio_Shared/Foley/Dino/Wavs/Scuff/Mimosa_Dino_Scuff_Hard_03}}
 {id=WavePlayer_13, properties {soundWaveAssetPtr=/CosmeticCompanions/Assets/Audio_Shared/Foley/Dino/Wavs/Scuff/Mimosa_Dino_Scuff_Hard_04}}
 {id=WavePlayer_2, properties {soundWaveAssetPtr=&Mimosa_Dino_Scuff_Hard_03}}
 {id=WavePlayer_3, properties {soundWaveAssetPtr=&Mimosa_Dino_Scuff_Hard_04}}
 {id=WavePlayer_4, properties {soundWaveAssetPtr=/CosmeticCompanions/Assets/Audio_Shared/Foley/Dino/Wavs/Scuff/Mimosa_Dino_Scuff_Hard_05}}
 {id=WavePlayer_5, properties {soundWaveAssetPtr=/CosmeticCompanions/Assets/Audio_Shared/Foley/Dino/Wavs/Scuff/Mimosa_Dino_Scuff_Hard_06}}
 {id=WavePlayer_6, properties {soundWaveAssetPtr=/CosmeticCompanions/Assets/Audio_Shared/Foley/Dino/Wavs/Scuff/Mimosa_Dino_Scuff_Hard_07}}
 {id=WavePlayer_7, properties {soundWaveAssetPtr=/CosmeticCompanions/Assets/Audio_Shared/Foley/Dino/Wavs/Scuff/Mimosa_Dino_Scuff_Hard_08}}
 {id=WavePlayer_8, properties {soundWaveAssetPtr=/CosmeticCompanions/Assets/Audio_Shared/Foley/Dino/Wavs/Scuff/Mimosa_Dino_Scuff_Hard_09}}
 {id=WavePlayer_9, properties {soundWaveAssetPtr=/CosmeticCompanions/Assets/Audio_Shared/Foley/Dino/Wavs/Scuff/Mimosa_Dino_Scuff_Hard_10}}
]
properties {
 attenuationSettings=/Game/Sounds/Attenuation/AS_Companion_Foley
 maxDistance=1200
 pitchMultiplier=1.5
 priority=0
 soundClassObject=&Fort_Creatures_Pets
 volumeMultiplier=0.2
 concurrencySet [/CosmeticCompanions/Assets/Audio_Shared/Foley/Biped/CONC_Mimosa_Biped_Footstep]
}
